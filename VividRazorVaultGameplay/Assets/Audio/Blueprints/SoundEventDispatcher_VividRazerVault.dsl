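; SoundEventDispatcher_VividRazerVault is a Blueprint that relays encounter events for the VividRazorVault gameplay content, named and filed as an audio event hub.
; Three custom events, EncounterBegin, EncounterEnd and AdvanceEncounterState, each call their matching event dispatcher.
; It holds no sound playback or other logic in this graph, so any response comes from whatever binds to those dispatchers.
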(event Custom|EncounterBegin
  (Default|CallOnEncounterBegin))

(event Custom|EncounterEnd
  (Default|CallOnEncounterEnd))

(event Custom|AdvanceEncounterState
  (Default|CallOnAdvanceEncounterState))
