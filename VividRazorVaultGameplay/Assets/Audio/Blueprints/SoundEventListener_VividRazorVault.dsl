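; Actor component that drives Vivid Razor bank vault audio for alert, opened, paused and encounter-advance events.
; BeginPlay finds or adds a sound library component, then, once the playlist is ready, the authority listens on the vault playspace's event router, filtered by location tag.
; AlertActive's OnRep schedules randomly delayed alarm sounds or stops the loop.
; Playspace lookup prefers a door owner's vault manager, covering volumes smaller than the vault.
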(event EventBeginPlay
  (bind _output_get (Variables|Transient|SetOwningActor (Components|GetOwner)))
  (Utilities|IsValid (Actor|GetComponentbyClass _output_get "/Script/FortniteGame.FortSoundLibraryComponent")
    (:Is%20Valid
      (Variables|Transient|SetSoundLibraryComponent (Actor|GetComponentbyClass _output_get "/Script/FortniteGame.FortSoundLibraryComponent")))
    (:Is%20Not%20Valid
      (bind _returnvalue (Game|AddComponentbyClass (Variables|Transient|GetOwningActor) "/VividRazorVaultGameplay/Assets/Audio/Blueprints/SoundLibraryComponent_VividRazorVault.SoundLibraryComponent_VividRazorVault_C"))
      (Variables|Transient|SetSoundLibraryComponent _returnvalue)))
  (SoundLibrary|AddLibraries (Variables|Transient|GetSoundLibraryComponent) (Variables|Default|GetSoundLibraries) :!class /Script/SoundLibrary.SoundLibraryComponent)
  (Class|FortAsyncActionCurrentPlaylistReady|PlaylistDataReadyAsync
    (:OnReady
      (CallFunction|StartListeners))))

(event Custom|StartListeners
  (bind _preconfiguredeventrouter (Variables|Transient|GetPreconfiguredEventRouter))
  (bind _eventrouterwait (Variables|Default|GetEventRouterWait))
  (Utilities|SwitchHasAuthority
    (:Authority
      (Utilities|IsValid _eventrouterwait
        (:Is%20Valid
          (AsyncAction|Cancel _eventrouterwait)
          (Utilities|IsValid _preconfiguredeventrouter
            (:Is%20Valid
              (Variables|Transient|SetVaultEventRouter _preconfiguredeventrouter))
            (:Is%20Not%20Valid
              (Variables|Transient|SetVaultEventRouter 0)
              (bind (_asynctaskproxy _eventrouter) (Class|AsyncActionWaitforPlayspaceEventRouter|WaitforPlayspaceEventRouter (CallFunction|GetContextActorforPlayspaceQuery) "/VividRazorVaultGameplay/Gameplay/Environmental/Manager/BP_VividRazorVaultPlayspace.BP_VividRazorVaultPlayspace_C" true true :!label _label)
                (:then
                  (Variables|Default|SetEventRouterWait _asynctaskproxy))
                (:OnEventRouterAvailable
                  (Variables|Transient|SetVaultEventRouter _eventrouter)))))
          (CallFunction|ListenforAlertStarted)
          (CallFunction|ListenforVaultOpened)
          (CallFunction|ListenforEventPaused))
        (:Is%20Not%20Valid
          _))
      (goto _label))))

(event Custom|ListenForAlertStarted
  (bind _alertlistener (Variables|Default|GetAlertListener))
  (Utilities|IsValid _alertlistener
    (:Is%20Valid
      (GameplayEventRouter|StopListeningtoEvent _alertlistener)
      (Variables|Default|SetAlertListener 0))
    (:Is%20Not%20Valid
      _))
  (bind (_asynctaskproxy _payload) (GameplayEventRouter|StartListeningToEvent (Variables|Transient|GetVaultEventRouter) "/VividRazorCoreGameplay/Gameplay/Structs/Struct_VividRazorCore_Vault_Alert_Payload.Struct_VividRazorCore_Vault_Alert_Payload")
    (:then
      (Variables|Default|SetAlertListener _asynctaskproxy))
    (:OnEventReceived
      (if (== (Utilities|Struct|BreakStructVividRazorCoreVaultAlertPayload _payload) (Variables|Default|GetVaultLocationTag))
        (CallFunction|OnAlertInitiated)
        (bind _output_get (Variables|Transient|SetEventDispatcher (Actor|GetComponentbyClass (Utilities|ResolveSoftReference (.VaultManagerActor_28_2F604354450CF5624125F284BA0BA87B (Utilities|Struct|BreakStructVividRazorCoreVaultAlertPayload _payload))) "/VividRazorVaultGameplay/Assets/Audio/Blueprints/SoundEventDispatcher_VividRazerVault.SoundEventDispatcher_VividRazerVault_C")))
        (Utilities|IsValid _output_get
          (:Is%20Valid
            (Default|BindEventtoOnAdvanceEncounterState _output_get (EventDispatchers|CreateEvent :!event Custom|OnAdvanceEncounterState))))))))

(event Custom|ListenForVaultOpened
  (bind _openedlistener (Variables|Default|GetOpenedListener))
  (Utilities|IsValid _openedlistener
    (:Is%20Valid
      (GameplayEventRouter|StopListeningtoStatefulEvent _openedlistener)
      (Variables|Default|SetOpenedListener 0))
    (:Is%20Not%20Valid
      _))
  (bind (_asynctaskproxy _payload) (GameplayEventRouter|StartListeningToStatefulEvent (Variables|Transient|GetVaultEventRouter) "/VividRazorCoreGameplay/Gameplay/Structs/Struct_VividRazorCore_Vault_Opened_Payload.Struct_VividRazorCore_Vault_Opened_Payload")
    (:then
      (Variables|Default|SetOpenedListener _asynctaskproxy))
    (:OnEventReceived
      (if (== (Utilities|Struct|BreakStructVividRazorCoreVaultOpenedPayload _payload) (Variables|Default|GetVaultLocationTag))
        (CallFunction|OnVaultCracked)
        (CallFunction|VaultOpens)))))

(event Custom|VaultOpens
  (Networking|CanExecuteCosmeticEvents
    (:True
      (bind _soundlibrarycomponent (Variables|Transient|GetSoundLibraryComponent)
        (:then
          (SoundLibrary|PlayEvent _soundlibrarycomponent (Variables|SoundLibrary|Tags|GetVaultOpens_Tag)))
        (:else)))))

(event Custom|OnAdvanceEncounterState
  (Networking|CanExecuteCosmeticEvents
    (:True
      (SoundLibrary|PlayEvent (Variables|Transient|GetSoundLibraryComponent) (Variables|SoundLibrary|Tags|GetAdvanceEncounter_Tag)))))

(event Custom|ListenForEventPaused
  (bind _pausedlistener (Variables|Default|GetPausedListener))
  (Utilities|IsValid _pausedlistener
    (:Is%20Valid
      (GameplayEventRouter|StopListeningtoStatefulEvent _pausedlistener)
      (Variables|Default|SetPausedListener 0))
    (:Is%20Not%20Valid
      _))
  (bind (_asynctaskproxy _payload) (GameplayEventRouter|StartListeningToStatefulEvent (Variables|Transient|GetVaultEventRouter) "/VividRazorCoreGameplay/Gameplay/Structs/Struct_VividRazorCore_Vault_Paused_Payload.Struct_VividRazorCore_Vault_Paused_Payload")
    (:then
      (Variables|Default|SetPausedListener _asynctaskproxy))
    (:OnEventReceived
      (if (== (Utilities|Struct|BreakStructVividRazorCoreVaultPausedPayload _payload) (Variables|Default|GetVaultLocationTag))
        (CallFunction|OnVaultCracked)
        (bind _soundlibrarycomponent (Variables|Transient|GetSoundLibraryComponent)
          (:then
            (SoundLibrary|PlayEvent _soundlibrarycomponent (Variables|SoundLibrary|Tags|GetAlarmReset_Tag)))
          (:else))))))

(event Custom|AlertWithDelay
  (if (Variables|EventVariables|GetAlertActive)
    (bind _soundlibrarycomponent (Variables|Transient|GetSoundLibraryComponent)
      (:then
        (SoundLibrary|PlayEvent _soundlibrarycomponent (Variables|SoundLibrary|Tags|GetAlertBegins_Tag))
        (SoundLibrary|PlayEvent _soundlibrarycomponent (Variables|SoundLibrary|Tags|GetAlertLoop_Tag)))
      (:else))))

(fn GetContextActorForPlayspaceQuery
  (bind _asbga_vivid_razor_vault_door_parent (Utilities|Casting|CastToBGA_VividRazorVault_DoorParent (Variables|Transient|GetOwningActor))
    (:then
      (Utilities|IsValid (Variables|/VividRazorVaultGameplay/Gameplay/Environmental/Vault/BGA_VividRazorVault_DoorParent.BGA_VividRazorVault_DoorParent_C|GetBGA_VaultManager _asbga_vivid_razor_vault_door_parent)
        (:Is%20Valid
          (return (Variables|/VividRazorVaultGameplay/Gameplay/Environmental/Vault/BGA_VividRazorVault_DoorParent.BGA_VividRazorVault_DoorParent_C|GetBGA_VaultManager _asbga_vivid_razor_vault_door_parent)))
        (:Is%20Not%20Valid
          _)))
    (:CastFailed
      _)) ; The Bank Vault Device's playspace volume can be configured to be smaller than the vault itself, which will result in the actor not counting as inside the volume\nIn this case try to use its Manager instead, since that will be centered inside the volume and works even if the volume's size is zero
  (return (Variables|Transient|GetOwningActor)))

(fn OnAlertInitiated
  (Variables|EventVariables|SetAlertActive true))

(fn OnRep_AlertActive
  (bind (_lowerbound _upperbound) (Utilities|Struct|BreakFloatRange (Variables|Alarm|GetAlarmDelay)))
  (bind (_lowerbound_type _lowerbound_value) (Utilities|Struct|BreakFloatRangeBound _lowerbound))
  (bind (_upperbound_type _upperbound_value) (Utilities|Struct|BreakFloatRangeBound _upperbound))
  (bind _alerttimer (Variables|Alarm|GetAlertTimer))
  (bind _alertloop_tag (Variables|SoundLibrary|Tags|GetAlertLoop_Tag))
  (Networking|CanExecuteCosmeticEvents
    (:True
      (if (Variables|EventVariables|GetAlertActive)
        (bind _returnvalue (Utilities|Time|SetTimerbyEvent (EventDispatchers|CreateEvent :!event AlertWithDelay) (Math|Float|Max(Float) (Math|Random|RandomFloatinRange _lowerbound_value _upperbound_value) 0.001)))
        (Variables|Alarm|SetAlertTimer _returnvalue)
        (elif (Utilities|Time|IsValidTimerHandle _alerttimer)
          (Utilities|Time|InvalidateTimerHandle _alerttimer)))
      (bind _soundlibrarycomponent (Variables|Transient|GetSoundLibraryComponent)
        (:then
          (if (SoundLibrary|IsEventActive _soundlibrarycomponent _alertloop_tag)
            (SoundLibrary|StopEvent _soundlibrarycomponent _alertloop_tag)))
        (:else)))))

(fn OnVaultCracked
  (Variables|EventVariables|SetAlertActive false))

(fn SpawnPatrollingGuards)
