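; Sound event listener for the Razor Vault door that plays audio for weakpoint spawn, ready, damage and break stages.
; Cosmetic-gated events pass Sound Library tags to a shared helper, and spawn also stores the current socket.
; A remote-only, DoOnce-gated health-change handler triggers the target sound.
; The helper attaches returned audio components at that socket and sets "islocal" from whether the door's last weakpoint damager is the local player.
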(event Custom|OnWeakpointHealthChanged
  (params (NewHealth :type Float%20%28double-precision%29))
  (Networking|SwitchHasAuthority
    (:Remote
      (if (< NewHealth 0.95)
        _)
      (Utilities|FlowControl|DoOnce false
        (:Completed
          (CallFunction|PlayEventatSocket (Variables|SoundLibrary|Tags|GetWeakpointTarget_Tag)))))))

(event Custom|WeakpointSpawn
  (params (WeakpointSocket :type Name))
  (Networking|CanExecuteCosmeticEvents
    (:True
      (Variables|Default|SetCurrentWeakpointSocket WeakpointSocket)
      (CallFunction|PlayEventatSocket (Variables|SoundLibrary|Tags|GetWeakpointSpawn_Tag)))))

(event Custom|WeakpointReady
  (Networking|CanExecuteCosmeticEvents
    (:True
      (CallFunction|PlayEventatSocket (Variables|SoundLibrary|Tags|GetWeakpointReady_Tag)))))

(event Custom|WeakpointDamaged
  (Networking|CanExecuteCosmeticEvents
    (:True
      (CallFunction|PlayEventatSocket (Variables|SoundLibrary|Tags|GetWeakpointDamaged_Tag)))))

(event Custom|WeakpointBreak
  (Networking|CanExecuteCosmeticEvents
    (:True
      (CallFunction|PlayEventatSocket (Variables|SoundLibrary|Tags|GetWeakpointBreak_Tag)))))

(fn PlayEventAtSocket
  (params (LibraryTag :type Sound%20Library%20Tag%20Structure))
  (bind _owningactor (Variables|Transient|GetOwningActor))
  (bind _returnvalue (SoundLibrary|PlayEvent (Variables|Transient|GetSoundLibraryComponent) LibraryTag))
  (for _array_element (.AudioComponents (Utilities|Struct|BreakSoundLibraryPlayResult _returnvalue))
    (Transformation|AttachComponentToComponent _array_element (Variables|/Script/Engine.Actor|GetRootComponent _owningactor) (Variables|Default|GetCurrentWeakpointSocket) "SnapToTarget" "SnapToTarget" "SnapToTarget" true)
    (bind _asbga_vivid_razor_vault_door_parent (Utilities|Casting|CastToBGA_VividRazorVault_DoorParent _owningactor))
    (bind _lastdamagerthisweakpoint (Variables|/VividRazorVaultGameplay/Gameplay/Environmental/Vault/BGA_VividRazorVault_DoorParent.BGA_VividRazorVault_DoorParent_C|GetLastDamagerThisWeakpoint _asbga_vivid_razor_vault_door_parent)
      (:then
        (Audio|Parameter|SetBooleanParameter _array_element "islocal" (Pawn|IsLocalPlayerController _lastdamagerthisweakpoint) :!class /Script/AudioExtensions.AudioParameterControllerInterface))
      (:else))))
