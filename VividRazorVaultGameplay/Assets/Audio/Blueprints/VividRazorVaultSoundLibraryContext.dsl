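; Sound library context for Vivid Razor Vault audio that gives alarm sounds extra mix and music-system handling.
; Play runs the parent, then for alarm-tagged events routes each returned component's volume to the alarm sidechain control bus.
; With a resolved group, it also registers them with the music event subsystem as diegetic NPC events.
; GetMusicGroup derives that group from NPC music component tags on the owner or attach parent, or from mapped vault location tags.
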(fn GetMusicGroup
  (params (SoundLibraryContextEventInput :type Sound%20Library%20Context%20Event%20Input%20Structure%20%28by%20ref%29))
  (bind _owner (LocalVariables|GetOwner))
  (bind _returnvalue (Actor|GetAttachParentActor _owner))
  (bind _returnvalue_2 (Actor|GetComponentbyClass _returnvalue "/FortNPCMusic/Audio/NPCMusicComponent_VividRazorVaults.NPCMusicComponent_VividRazorVaults_C"))
  (bind _owner_2 (LocalVariables|GetOwner))
  (bind _returnvalue_3 (Actor|GetComponentbyClass _owner_2 "/VividRazorVaultGameplay/Assets/Audio/Blueprints/SoundEventListener_VividRazorVault.SoundEventListener_VividRazorVault_C"))
  (bind (_value _returnvalue_4) (Utilities|Map|Find (Variables|Default|GetLocationTagMap) (LocalVariables|GetLocationTag)))
  (bind _pawngrouptags (Variables|/FortNPCMusic/Audio/NPCMusicComponent.NPCMusicComponent_C|GetPawnGroupTags _returnvalue_2))
  (bind _output_get (LocalVariables|SetOwner (Utilities|Struct|BreakSoundLibraryContextEventInput SoundLibraryContextEventInput)))
  (Utilities|IsValid _output_get
    (:Is%20Valid
      (Utilities|IsValid (Actor|GetComponentbyClass _output_get "/FortNPCMusic/Audio/NPCMusicComponent_VividRazorVaults.NPCMusicComponent_VividRazorVaults_C")
        (:Is%20Valid
          (if (and (Variables|/FortNPCMusic/Audio/NPCMusicComponent_VividRazorVaults.NPCMusicComponent_VividRazorVaults_C|GetEnable%20Music (Actor|GetComponentbyClass _output_get "/FortNPCMusic/Audio/NPCMusicComponent_VividRazorVaults.NPCMusicComponent_VividRazorVaults_C")) (Utilities|GameplayTagContainerNotEmpty (Variables|/FortNPCMusic/Audio/NPCMusicComponent.NPCMusicComponent_C|GetPawnGroupTags (Actor|GetComponentbyClass _output_get "/FortNPCMusic/Audio/NPCMusicComponent_VividRazorVaults.NPCMusicComponent_VividRazorVaults_C"))))
            (return true (Utilities|String|StringToName (GameplayTags|GetDebugStringfromGameplayTag (Utilities|Array|Get(acopy) (GameplayTags|BreakGameplayTagContainer (Variables|/FortNPCMusic/Audio/NPCMusicComponent.NPCMusicComponent_C|GetPawnGroupTags (Actor|GetComponentbyClass _output_get "/FortNPCMusic/Audio/NPCMusicComponent_VividRazorVaults.NPCMusicComponent_VividRazorVaults_C")))))))
            (else
              _)))
        (:Is%20Not%20Valid
          _))
      (Utilities|IsValid _returnvalue
        (:Is%20Valid
          (Utilities|IsValid _returnvalue_2
            (:Is%20Valid
              (if (and (Variables|/FortNPCMusic/Audio/NPCMusicComponent_VividRazorVaults.NPCMusicComponent_VividRazorVaults_C|GetEnable%20Music _returnvalue_2) (Utilities|GameplayTagContainerNotEmpty _pawngrouptags))
                (return true (Utilities|String|StringToName (GameplayTags|GetDebugStringfromGameplayTag (Utilities|Array|Get(acopy) (GameplayTags|BreakGameplayTagContainer _pawngrouptags)))))
                (else
                  _)))
            (:Is%20Not%20Valid
              _))
          (bind _asbp_vivid_razor_vault_alert_light (Utilities|Casting|CastToBP_VividRazorVault_AlertLight _owner_2)
            (:then
              (LocalVariables|SetLocationTag (Variables|/VividRazorVaultGameplay/Gameplay/Environmental/Lights/BP_VividRazorVault_Prop_Parent.BP_VividRazorVault_Prop_Parent_C|GetVaultLocationTag _asbp_vivid_razor_vault_alert_light)))
            (:CastFailed
              (Utilities|IsValid _returnvalue_3
                (:Is%20Valid
                  (LocalVariables|SetLocationTag (Variables|/VividRazorVaultGameplay/Assets/Audio/Blueprints/SoundEventListener_VividRazorVault.SoundEventListener_VividRazorVault_C|GetVaultLocationTag _returnvalue_3)))
                (:Is%20Not%20Valid
                  (return true "None")))))
          (return _returnvalue_4 (Utilities|String|StringToName (GameplayTags|GetDebugStringfromGameplayTag _value))))
        (:Is%20Not%20Valid
          _)))
    (:Is%20Not%20Valid
      (return false "None"))))

(fn Play
  (params (InEventData :type Sound%20Library%20Context%20Event%20Input%20Structure%20%28by%20ref%29))
  (bind (_outcomponents _returnvalue) (|CallParentFunction InEventData))
  (if (GameplayTags|MatchesAnyTags (Utilities|Struct|BreakSoundLibraryContextEventInput InEventData) (LocalVariables|GetAlarmTagContainer) true)
    (bind (_success _group) (CallFunction|GetMusicGroup InEventData))
    (LocalVariables|SetUseMusicGroup _success)
    (LocalVariables|SetMusicGroup _group)
    (for _array_element _outcomponents
      (Audio|Components|Audio|AddModulationRouting _array_element (Utilities|Set|MakeSet "/VividRazorVaultGameplay/Assets/Audio/Vault/settings/CB_VaultAlarms_SideChain.CB_VaultAlarms_SideChain") "Volume")
      (if (LocalVariables|GetUseMusicGroup)
        (Audio|AddEventbyComponent (GameInstanceSubsystems|GetMusicEventSubsystem) _array_element "(TagName=\"MusicEventSubsystem.Event.NPC.Diegetic\")" "(GameplayTags=)" 0 (LocalVariables|GetMusicGroup))))
    (else
      _))
  (return _outcomponents _returnvalue))
